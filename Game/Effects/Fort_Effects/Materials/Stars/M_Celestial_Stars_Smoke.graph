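# Unlit translucent Niagara sprite material for celestial smoke puffs filled with galaxy nebula and twinkling stars.
# A 2x2 inky smoke flipbook, phased and eroded by dynamic parameters, shapes opacity and slightly offsets the nebula.
# Rotating camera-vector triplanar nebula and star layers, with stars masked by a panning copy, drive color-tinted emissive and add to opacity.
# Near-camera fade shrinks sprites toward particle center and, with depth fade, softens alpha.
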Material /Game/Effects/Fort_Effects/Materials/Stars/M_Celestial_Stars_Smoke
graph {
 properties {
  bAutomaticallySetUsageInEditor=false
  bUsedWithNiagaraSprites=true
  bUsedWithStaticMesh=false
  blendMode=BLEND_Translucent
  pixelDepthOffsetMode=PDOM_Legacy
  shadingModel=MSM_Unlit
  usedShadingModels=Unlit
 }
 nodes [
  n1 = Power(Base=MaterialFunctionCall(Distance=ScalarParameter() {defaultValue=150, parameterName=ShrinkDistance}) {materialFunction=&Near_Camera_Fade}.Result) {constExponent=3}
  LinearInterpolate(A=Subtract(A=ParticlePositionWS(), B=WorldPosition().XYZ), Alpha=@n1) {constB=0}
  n2 = ParticleColor()
  n3 = ScalarParameter() {defaultValue=0.02, group=Galaxy, parameterName=Galaxy Rotation Speed_1}
  n4 = ScalarParameter() {defaultValue=-0.02, group=Galaxy, parameterName=Galaxy Rotation Speed_2}
  n5 = Multiply(
   A=Time()
   B=AppendVector(A=AppendVector(A=@n3, B=@n4), B=ScalarParameter() {defaultValue=0.1, group=Galaxy, parameterName=Galaxy Rotation Speed_3})
  )
  n6 = DynamicParameter() {paramNames [Random,Erode,PanSpeed,MasterAlpha]}
  n7 = TextureCoordinate()
  n8 = Constant() {r=0.5}
  n9 = Constant() {r=2}
  n10 = Add(
   A=Multiply(A=SphereMask(A=@n7, B=@n8, Radius=@n8) {hardnessPercent=0}) {constB=0.15}
   B=TextureSample(UVs=MaterialFunctionCall("Animation  Phase (0-1)"=@n6.Random, Number of Rows=@n9, Number of Columns=@n9) {materialFunction=&FlipBook}.UVs) {texture=/Game/Effects/Fort_Effects/Textures/T_InkySmoke_Flipbook}.G
  )
  n11 = Add(A=@n6.Erode, B=@n10)
  n12 = Clamp(None=Multiply(A=LinearInterpolate(Alpha=@n11) {constA=-3}, B=@n10))
  n13 = MaterialFunctionCall!(
   CameraVector
   Rotation Speed=@n5
   Texture=TextureObject() {texture=/Game/Characters/Player/Male/Medium/Bodies/M_Med_Soldier_04/Skins/Galaxy/Textures/T_FN_Nebula}
   Tiling=ScalarParameter() {defaultValue=0.5, group=Galaxy, parameterName=GalaxyTiling}
   Offset=Multiply(A=@n12) {constB=0.01}
   AxisFadeContrast
  ) {materialFunction=/Game/Weapons/WeaponSkins/MaterialFunction/TriplanarCameraVector_Rotated}
  n14 = TextureObject() {samplerType=SAMPLERTYPE_Masks, texture=/Game/Characters/Player/Male/Medium/Bodies/M_Med_Soldier_04/Skins/Galaxy/Textures/Celestial_T_Stars}
  n15 = MaterialFunctionCall(Rotation Speed=Multiply(A=@n5) {constB=-1.5}, Texture=@n14, Tiling=ScalarParameter() {defaultValue=7, parameterName=StarTiling}) {materialFunction=&TriplanarCameraVector_Rotated}
  n16 = TextureSample(UVs=Panner(Coordinate=TextureCoordinate() {uTiling=2, vTiling=2}) {speedX=0.2, speedY=-0.2}) {samplerType=SAMPLERTYPE_Masks, texture=&Celestial_T_Stars}
  n17 = Multiply(A=Power(Base=ComponentMask(None=@n15.Result) {r=true}) {constExponent=1.5}, B=@n16.G)
  n18 = ScalarParameter() {defaultValue=1, parameterName=Star Opacity}
  n19 = Multiply(A=Multiply(A=Add(A=@n13.Result) {constB=0.2}, B=Multiply(A=@n17) {constB=900}), B=@n18)
  MaterialFunctionCall(Emissive=Multiply(A=@n2.RGB, B=Add(A=@n13.Result, B=@n19))) {materialFunction=&MF_FX_EmissiveAdjustments}
  MaterialFunctionCall() {materialFunction=&DebugTimeSine}
  Multiply(A=MaterialFunctionCall() {materialFunction=&RadialGradientExponential}.RadialGradientExponential) {constB=2}
  n20 = Multiply(A=@n12) {constB=0.7}
  n21 = Clamp(None=@n17)
  n22 = Constant() {r=0.5}
  n23 = SphereMask(A=TextureCoordinate(), B=@n22, Radius=@n22, Hardness=Constant() {r=0.6})
  n24 = Multiply(
   A=LinearInterpolate(
    A=@n21
    B=Multiply(A=@n23, B=@n21)
    Alpha=ScalarParameter() {desc=Modulates star opacity to prevent stars from reaching the end of the sprite, parameterName=soft Star edge}
   )
   B=@n18
  )
  Multiply(A=Multiply(A=@n2.A, B=Add(A=@n20, B=@n24)), B=DepthFade(Opacity=@n1) {fadeDistanceDefault=12})
 ]
}
