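# M_Star is an unlit translucent material for Cascade and Niagara mesh particles, likely for star effects given its name.
# Particle Color RGB feeds the Emissive input of the shared MF_FX_EmissiveAdjustments function, so each emitter controls the tint.
# An orange HDR constant vector sits unused in the graph.
# No texture sampling or opacity logic appears, and the graph shows no links to the material outputs.
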Material /Game/Effects/Fort_Effects/Materials/Stars/M_Star
graph {
 properties {
  bUsedWithMeshParticles=true
  bUsedWithNiagaraMeshParticles=true
  blendMode=BLEND_Translucent
  pixelDepthOffsetMode=PDOM_Legacy
  shadingModel=MSM_Unlit
  usedShadingModels=Unlit
 }
 nodes [
  Constant3Vector() {constant=<rgba 3 1.5 0.5 0>}
  MaterialFunctionCall(Emissive=ParticleColor().RGB) {materialFunction=&MF_FX_EmissiveAdjustments}
 ]
}
